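# Sound cue for a player entering a phone booth hideable, playing a third-person door-open sound.
# A Random node picks evenly between two open-sound variations and is excluded from branch culling.
# The cue routes through the gadgets and items sound class and spatializes with the shared mono trap attenuation preset.
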SoundCue /Game/Sounds/Fort_Hideables/PhoneBooth/Hideable_PhoneBooth_Enter_Cue
root=Random_1
nodes [
 {id=Random_1, children [WavePlayer_0,WavePlayer_4], properties {bShouldExcludeFromBranchCulling=true, weights [1,1]}}
 {id=WavePlayer_0, properties {soundWaveAssetPtr=&/Hideable_PhoneBooth_Open_01_3P}}
 {id=WavePlayer_4, properties {soundWaveAssetPtr=&/Hideable_PhoneBooth_Open_02_3P}}
]
properties {
 attenuationSettings=/Game/Sounds/Attenuation/Fort_Traps_Mono_Att
 maxDistance=6000
 soundClassObject=&Fort_Gadgets_Items
 volumeMultiplier=1.0714
}
