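# Sound cue for the third-person phone booth hideable exit sound.
# A Random node picks between two equally weighted third-person door-close waves and is excluded from branch culling.
# It plays in the Fort_Gadgets_Items sound class at reduced volume, using the shared mono trap attenuation for spatial falloff.
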SoundCue /Game/Sounds/Fort_Hideables/PhoneBooth/Hideable_PhoneBooth_Exit_Cue_3P
root=Random_0
nodes [
 {id=Random_0, children [WavePlayer_0,WavePlayer_3], properties {bShouldExcludeFromBranchCulling=true, weights [1,1]}}
 {id=WavePlayer_0, properties {soundWaveAssetPtr=&/Hideable_PhoneBooth_Close_01_3P}}
 {id=WavePlayer_3, properties {soundWaveAssetPtr=&/Hideable_PhoneBooth_Close_02_3P}}
]
properties {
 attenuationSettings=/Game/Sounds/Attenuation/Fort_Traps_Mono_Att
 maxDistance=6000
 soundClassObject=&Fort_Gadgets_Items
 volumeMultiplier=0.4286
}
