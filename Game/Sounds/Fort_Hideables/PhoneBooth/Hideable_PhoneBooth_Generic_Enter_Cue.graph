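# Sound cue for the generic phone booth hideable prop, played by its name when a player enters the booth.
# A single wave player outputs the third-person phone booth open sound directly, with no randomization or mixing.
# It uses the mono traps attenuation asset for spatial falloff and routes to the gadgets and items sound class.
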SoundCue /Game/Sounds/Fort_Hideables/PhoneBooth/Hideable_PhoneBooth_Generic_Enter_Cue
root=WavePlayer_2
nodes [{id=WavePlayer_2, properties {soundWaveAssetPtr=&/HideableProp_PhoneBooth_Generic_Open_3P}}]
properties {
 attenuationSettings=/Game/Sounds/Attenuation/Fort_Traps_Mono_Att
 maxDistance=6000
 soundClassObject=&Fort_Gadgets_Items
 volumeMultiplier=1
}
