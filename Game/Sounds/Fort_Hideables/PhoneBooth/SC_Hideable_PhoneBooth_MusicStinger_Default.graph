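# Sound cue for the default music stinger of the phone booth hideable.
# A mixer starts the open stinger immediately alongside a random pick between two close stingers, each held back about two seconds by its own fixed delay, so one cue covers the timed open-then-close sequence.
# It uses the shared mono trap attenuation and the gadgets/items sound class.
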SoundCue /Game/Sounds/Fort_Hideables/PhoneBooth/SC_Hideable_PhoneBooth_MusicStinger_Default
root=Mixer_0
nodes [
 {id=Delay_0, children [WavePlayer_0], properties {delayMax=2.2, delayMin=2.2}}
 {id=Delay_1, children [WavePlayer_3], properties {delayMax=2.1, delayMin=2.1}}
 {id=Mixer_0, children [WavePlayer_2,Random_0], properties {inputVolume [1,1]}}
 {id=Random_0, children [Delay_0,Delay_1], properties {weights [1,1]}}
 {id=WavePlayer_0, properties {soundWaveAssetPtr=&/Hideable_PhoneBooth_Close_01_MX}}
 {id=WavePlayer_2, properties {soundWaveAssetPtr=&/Hideable_PhoneBooth_Open_01_MX}}
 {id=WavePlayer_3, properties {soundWaveAssetPtr=&/Hideable_PhoneBooth_Close_02_MX}}
]
properties {
 attenuationSettings=/Game/Sounds/Attenuation/Fort_Traps_Mono_Att
 maxDistance=6000
 soundClassObject=&Fort_Gadgets_Items
 volumeMultiplier=1.0714
}
